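# Landscape material function that blends arid grass color into the Standard and Low material attribute inputs, changing only their BaseColor.
# Grass BaseColor is multiplied by the MPC_Landscape tint and lerped in by a HeightLerp of the mask, using grass Opacity as height.
# A third output returns the mask divided by a scalar parameter and clamped.
# Full-layer blends, including one with quality and feature-level switched alpha, feed no output.
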MaterialFunction /FigmentEnvironment/LM_8_1_1/Athena/Environments/Landscape/MaterialFunctions/Arid/MF_Athena_AridGrass
graph {nodes [
  n1 = FunctionInput() {inputName=Material_Standard, inputType=FunctionInput_MaterialAttributes}
  n2 = GetMaterialAttributes(None=@n1) {attributeGetTypes [#1]}
  n3 = FunctionInput() {inputName=Grass, inputType=FunctionInput_MaterialAttributes}
  n4 = GetMaterialAttributes(None=@n3) {attributeGetTypes [#1]}
  n5 = ComponentMask(None=CollectionParameter() {collection=/FigmentEnvironment/LM_8_1_1/Athena/Environments/Landscape/MPC/MPC_Landscape, parameterName=Arid_Grass_Tint}) {b=true, g=true, r=true}
  n6 = Multiply(A=@n4.BaseColor, B=@n5)
  n7 = FunctionInput() {inputName=Mask, inputType=FunctionInput_Scalar, sortPriority=2}
  n8 = MaterialFunctionCall(
   Transition Phase=@n7
   Height Texture=MaterialFunctionCall!(Mat attributes=@n3) {materialFunction=/Engine/Functions/Engine_MaterialFunctions02/MatLayerBlend_BreakOpacity}.Result
   Contrast=ScalarParameter() {parameterName=Arid_Grass_Contrast}
  ) {materialFunction=&HeightLerp}
  FunctionOutput(None=SetMaterialAttributes!(MaterialAttributes=@n2.MaterialAttributes, BaseColor=LinearInterpolate(A=@n2.BaseColor, B=@n6, Alpha=@n8.Alpha)) {attributeSetTypes [#1]}) {outputName=Material_Standard}
  n9 = FunctionInput() {inputName=Material_BaseColor, inputType=FunctionInput_MaterialAttributes, sortPriority=1}
  n10 = GetMaterialAttributes(None=@n9) {attributeGetTypes [#1]}
  n11 = Multiply(A=@n4.BaseColor, B=@n5)
  FunctionOutput(None=SetMaterialAttributes(MaterialAttributes=@n10.MaterialAttributes, BaseColor=LinearInterpolate(A=@n10.BaseColor, B=@n11, Alpha=@n8.Alpha)) {attributeSetTypes [#1]}) {outputName=Material_Low, sortPriority=1}
  n12 = ScalarParameter() {defaultValue=1.5, parameterName=AridGrassMob_MP}
  FunctionOutput(None=Clamp(None=Divide(A=@n7, B=@n12))) {outputName=ColorMap_MaskAdjustment, sortPriority=2}
  MaterialFunctionCall!(
   Base Material=@n1
   Top Material=MaterialFunctionCall!(In=@n3, Color=@n5) {materialFunction=/FigmentEnvironment/LM_8_1_1/Athena/Environments/Landscape/MaterialFunctions/MF_Athena_Material_BaseMultiply}.Result
   Alpha=@n8.Alpha
  ) {materialFunction=/Engine/Functions/MaterialLayerFunctions/MatLayerBlend_Standard}
  n13 = QualitySwitch(Default=@n8.Alpha, Low=Clamp(None=Multiply(A=@n7, B=@n12)), High=@n8.Alpha, Medium=@n8.Alpha) {desc="Use Height lerp for Medium, but simple lerp for Mobile/Low"}
  MaterialFunctionCall(Base Material=@n9, Alpha=FeatureLevelSwitch(Default=@n13, ES2=@n7, ES3_1=@n7, SM4_REMOVED=@n13, SM5=@n13, SM6=@n13)) {materialFunction=&MatLayerBlend_Standard}
 ]}
